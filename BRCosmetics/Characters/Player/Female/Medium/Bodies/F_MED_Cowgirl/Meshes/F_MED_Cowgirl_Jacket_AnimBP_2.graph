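# Animation Blueprint that adds secondary motion to the jacket, scarf, tassels and spurs of the female Cowgirl cosmetic.
# When rigid body is enabled, the jacket simulates with separate skydive and regular physics assets under a gravity override.
# Otherwise, movement-driven bone offsets, look-ats and trail controllers pose it procedurally from a cached 'prerigid' pose.
# The blended result then feeds constraints and spring-based AnimDynamics for spurs, tassels and scarf.
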AnimBlueprint /BRCosmetics/Characters/Player/Female/Medium/Bodies/F_MED_Cowgirl/Meshes/F_MED_Cowgirl_Jacket_AnimBP_2
target_skeleton=/Game/Characters/Player/Female/Medium/Bodies/F_MED_Cowgirl/Meshes/F_MED_Cowgirl_Skeleton
graphs [
 AnimGraph {nodes [
   Component To Local()
   "Constraint - dyn_scarf_arm_01 (PP)"!(ComponentPose, Alpha=1, ConstraintWeights_0=0.7179, ConstraintWeights_1:float=0.2821) -> (Pose) {
    alphaBoolBlend {blendOption=Linear}
    boneToModify {boneName=dyn_scarf_arm_01}
    constraintSetup [{targetBone {boneName=upperarm_twist_01_l}}, {targetBone {boneName=clavicle_l}}]
    constraintWeights [1,1]
   }
   "Constraint - dyn_scarf_arm_01 (R)"(Alpha=0.5, ConstraintWeights_0=1) {
    alphaBoolBlend {blendOption=Linear}
    boneToModify {boneName=dyn_scarf_arm_01}
    constraintSetup [{transformType=Rotation, targetBone {boneName=upperarm_twist_01_l}}]
    constraintWeights [1]
   }
   n1 = Get BwdMovement_scarf_0() -> (BwdMovement_scarf_0:bool)
   "Constraint - dyn_scarf_arm_bk_01 (P)"!(ComponentPose, bAlphaBoolEnabled:bool=@n1, ConstraintWeights_0=1) -> (Pose) {
    alphaInputType=Bool
    alphaBoolBlend {blendInTime=0.1, blendOption=Linear, blendOutTime=0.5}
    alphaScaleBiasClamp {bInterpResult=true, interpSpeedDecreasing=50}
    boneToModify {boneName=dyn_scarf_arm_bk_01}
    constraintSetup [{targetBone {boneName=upperarm_twist_01_l}}]
    constraintWeights [1]
   }
   Local To Component()
   n2 = Get isNotCrouchRunning() -> (isNotCrouchRunning:double)
   "Transform (Modify) Bone - Bone: dyn_jacketBack_03_l"(Alpha=@n2, Translation="0,0,0", Rotation="25,0,0.000", Scale="1,1,1") {
    rotationMode=BMM_Additive
    rotationSpace=BCS_BoneSpace
    alphaBoolBlend {blendOption=Linear}
    alphaScaleBiasClamp {bInterpResult=true, interpSpeedDecreasing=2, interpSpeedIncreasing=5}
    boneToModify {boneName=dyn_jacketBack_03_l}
    rotation {pitch=25}
   }
   "Transform (Modify) Bone - Bone: dyn_jacketFront_mstr_r"(Alpha=Get thigh_rot_up_r() -> (thigh_rot_up_r:double), Translation="0,0,3", Rotation="0,0,0", Scale="1,1,1") {
    rotationMode=BMM_Additive
    rotationSpace=BCS_BoneSpace
    translationMode=BMM_Additive
    translationSpace=BCS_BoneSpace
    alphaBoolBlend {blendOption=Linear}
    boneToModify {boneName=dyn_jacketFront_mstr_r}
    translation {z=3}
   }
   n3 = Get in vehicle scale alpha() -> (in vehicle scale alpha:double)
   n4 = "Transform (Modify) Bone - Bone: dyn_jacketFront_01_l"(Alpha=@n3, Translation="0,0,0", Rotation="0,0,0", Scale="0.1,0.1,0.1") {scaleMode=BMM_Replace, alphaBoolBlend {blendOption=Linear}, boneToModify {boneName=dyn_jacketFront_01_l}, scale=<0.1 0.1 0.1>}
   n5 = "Transform (Modify) Bone - Bone: dyn_skirt_frdiag_01_l"(ComponentPose=@n4, Alpha=@n3, Translation="0,0,0", Rotation="0,0,0", Scale="0.1,0.1,0.1") {scaleMode=BMM_Replace, alphaBoolBlend {blendOption=Linear}, boneToModify {boneName=dyn_skirt_frdiag_01_l}, scale=<0.1 0.1 0.1>}
   n6 = "Transform (Modify) Bone - Bone: dyn_skirt_out_01_l"(ComponentPose=@n5, Alpha=@n3, Translation="0,0,0", Rotation="0,0,0", Scale="0.1,0.1,0.1") {scaleMode=BMM_Replace, alphaBoolBlend {blendOption=Linear}, boneToModify {boneName=dyn_skirt_out_01_l}, scale=<0.1 0.1 0.1>}
   n7 = "Transform (Modify) Bone - Bone: dyn_skirt_bk_01_l"(ComponentPose=@n6, Alpha=@n3, Translation="0,0,0", Rotation="0,0,0", Scale="0.1,0.1,0.1") {scaleMode=BMM_Replace, alphaBoolBlend {blendOption=Linear}, boneToModify {boneName=dyn_skirt_bk_01_l}, scale=<0.1 0.1 0.1>}
   n8 = "Transform (Modify) Bone - Bone: dyn_skirt_mid_01"(ComponentPose=@n7, Alpha=@n3, Translation="0,0,0", Rotation="0,0,0", Scale="0.1,0.1,0.1") {scaleMode=BMM_Replace, alphaBoolBlend {blendOption=Linear}, boneToModify {boneName=dyn_skirt_mid_01}, scale=<0.1 0.1 0.1>}
   n9 = "Transform (Modify) Bone - Bone: dyn_skirt_fr_01_r"(ComponentPose=@n8, Alpha=@n3, Translation="0,0,0", Rotation="0,0,0", Scale="0.1,0.1,0.1") {scaleMode=BMM_Replace, alphaBoolBlend {blendOption=Linear}, boneToModify {boneName=dyn_skirt_fr_01_r}, scale=<0.1 0.1 0.1>}
   n10 = "Transform (Modify) Bone - Bone: dyn_skirt_frdiag_01_r"(ComponentPose=@n9, Alpha=@n3, Translation="0,0,0", Rotation="0,0,0", Scale="0.1,0.1,0.1") {scaleMode=BMM_Replace, alphaBoolBlend {blendOption=Linear}, boneToModify {boneName=dyn_skirt_frdiag_01_r}, scale=<0.1 0.1 0.1>}
   n11 = "Transform (Modify) Bone - Bone: dyn_skirt_out_01_r"(ComponentPose=@n10, Alpha=@n3, Translation="0,0,0", Rotation="0,0,0", Scale="0.1,0.1,0.1") {scaleMode=BMM_Replace, alphaBoolBlend {blendOption=Linear}, boneToModify {boneName=dyn_skirt_out_01_r}, scale=<0.1 0.1 0.1>}
   "Transform (Modify) Bone - Bone: dyn_skirt_bk_01_r"(ComponentPose=@n11, Alpha=@n3, Translation="0,0,0", Rotation="0,0,0", Scale="0.1,0.1,0.1") {scaleMode=BMM_Replace, alphaBoolBlend {blendOption=Linear}, boneToModify {boneName=dyn_skirt_bk_01_r}, scale=<0.1 0.1 0.1>}
   RigidBody() {
    bClampLinearTranslationLimitToRefPose=true
    bOverrideWorldGravity=true
    overridePhysicsAsset=/Game/Characters/Player/Female/Medium/Bodies/F_MED_Cowgirl/Meshes/F_MED_Cowgirl_T456_Tassels_Physics
    simulationSpace=BaseBoneSpace
    alphaBoolBlend {blendOption=Linear}
    baseBoneRef {boneName=root}
    componentAppliedLinearAccClamp=<1500 1500 1500>
    componentLinearAccScale=<0.1 0.1 0.5>
    componentLinearVelScale=<0.1 0.1 0.5>
    overrideWorldGravity {z=-540}
   }
   n12 = "Transform (Modify) Bone - Bone: dyn_jacketBack_mstr_r"(ComponentPose=Local To Component(LocalPose=Use cached pose 'prerigid'()), Alpha=1, Translation="0,-6,0", Rotation="0,0,0", Scale="1,1,1") {translationMode=BMM_Additive, alphaBoolBlend {blendOption=Linear}, boneToModify {boneName=dyn_jacketBack_mstr_r}, translation {y=-6}}
   n13 = "Transform (Modify) Bone - Bone: dyn_jacketBack_mstr_l"(ComponentPose=@n12, Alpha=1, Translation="0,-6,0", Rotation="0,0,0", Scale="1,1,1") {translationMode=BMM_Additive, alphaBoolBlend {blendOption=Linear}, boneToModify {boneName=dyn_jacketBack_mstr_l}, translation {y=-6}}
   n14 = "Transform (Modify) Bone - Bone: dyn_jacketSide_mstr_r"(ComponentPose=@n13, Alpha=1, Translation="-3,-5,0", Rotation="0,0,0", Scale="1,1,1") {translationMode=BMM_Additive, alphaBoolBlend {blendOption=Linear}, boneToModify {boneName=dyn_jacketSide_mstr_r}, translation=<-3 -5>}
   n15 = "Transform (Modify) Bone - Bone: dyn_jacketSide_mstr_l"(ComponentPose=@n14, Alpha=1, Translation="3,-5,0", Rotation="0,0,0", Scale="1,1,1") {translationMode=BMM_Additive, alphaBoolBlend {blendOption=Linear}, boneToModify {boneName=dyn_jacketSide_mstr_l}, translation=<3 -5>}
   n16 = RigidBody!(ComponentPose=@n15, Alpha, OverrideWorldGravity:Vector=Get GravityOverride(), ExternalForce) -> (Pose) {
    bClampLinearTranslationLimitToRefPose=true
    bOverrideWorldGravity=true
    overridePhysicsAsset=&/F_MED_Cowgirl_Jacket_skydive_Physics_2
    simulationSpace=BaseBoneSpace
    alphaBoolBlend {blendOption=Linear}
    baseBoneRef {boneName=dyn_root_simspace}
    componentAppliedLinearAccClamp=<3000 3000 3000>
    componentLinearAccScale=<0.25 0.25 0.4>
    componentLinearVelScale=<0.2 0.2 0.2>
   }
   n17 = RigidBody(ComponentPose=Local To Component(LocalPose=Use cached pose 'prerigid'()), OverrideWorldGravity=Get GravityOverride()) {
    bClampLinearTranslationLimitToRefPose=true
    bOverrideWorldGravity=true
    overridePhysicsAsset=&/F_MED_Cowgirl_Jacket_Physics_2
    simulationSpace=BaseBoneSpace
    alphaBoolBlend {blendOption=Linear}
    baseBoneRef {boneName=dyn_root_simspace}
    componentAppliedLinearAccClamp=<3000 3000 3000>
    componentLinearAccScale=<0.25 0.25 0.4>
    componentLinearVelScale=<0.2 0.2 0.2>
   }
   n18 = Blend Poses by bool(BlendPose_0=Component To Local(ComponentPose=@n16), BlendPose_1=Component To Local(ComponentPose=@n17), bActiveValue=Get bIsSkyDiving()) {blendType=Linear}
   n19 = Get thigh_rot_inr_r() -> (thigh_rot_inr_r:double)
   n20 = Look At - Bone: dyn_jacketFront_mstr_r(ComponentPose=Local To Component(LocalPose=Use cached pose 'prerigid'()), Alpha=@n19) {
    alphaBoolBlend {blendOption=Linear}
    alphaScaleBiasClamp {bInterpResult=true}
    boneToModify {boneName=dyn_jacketFront_mstr_r}
    lookAtLocation=<0 -12 5>
    lookAtTarget {boneReference {boneName=calf_r}}
    lookAt_Axis {axis=<1 0>}
    lookUp_Axis {axis {x=0}}
   }
   n21 = Look At - Bone: dyn_jacketSide_mstr_r(ComponentPose=@n20, Alpha=@n19) {
    alphaBoolBlend {blendOption=Linear}
    alphaScaleBiasClamp {bInterpResult=true}
    boneToModify {boneName=dyn_jacketSide_mstr_r}
    lookAtLocation {x=0, z=11}
    lookAtTarget {boneReference {boneName=calf_r}}
    lookAt_Axis {axis=<1 0>}
    lookUp_Axis {axis {x=0}}
   }
   n22 = Get thigh_rot_up_r()
   n23 = "Transform (Modify) Bone - Bone: dyn_jacketFront_01_r"(ComponentPose=@n21, Alpha=@n22, Translation="0,0,0", Rotation="9.096,4.987,10.829", Scale="1,1,1") {
    rotationMode=BMM_Additive
    rotationSpace=BCS_BoneSpace
    translationMode=BMM_Additive
    translationSpace=BCS_BoneSpace
    alphaBoolBlend {blendOption=Linear}
    boneToModify {boneName=dyn_jacketFront_01_r}
    rotation=<rot 9.096 4.987 10.829>
   }
   n24 = "Transform (Modify) Bone - Bone: dyn_jacketSide_01_r"(ComponentPose=@n23, Alpha=@n22, Translation="0,0,0", Rotation="5.000,0.000,0.000", Scale="1,1,1") {
    rotationMode=BMM_Additive
    rotationSpace=BCS_BoneSpace
    translationMode=BMM_Additive
    translationSpace=BCS_BoneSpace
    alphaBoolBlend {blendOption=Linear}
    boneToModify {boneName=dyn_jacketSide_01_r}
    rotation {pitch=5}
   }
   n25 = Get thigh_rot_inr_l() -> (thigh_rot_inr_l:double)
   n26 = Look At - Bone: dyn_jacketFront_mstr_l(ComponentPose=@n24, Alpha=@n25) {
    alphaBoolBlend {blendOption=Linear}
    alphaScaleBiasClamp {bInterpResult=true}
    boneToModify {boneName=dyn_jacketFront_mstr_l}
    lookAtLocation=<0 12 -5>
    lookAtTarget {boneReference {boneName=calf_l}}
    lookAt_Axis {axis=<1 0>}
    lookUp_Axis {axis {x=0}}
   }
   n27 = Look At - Bone: dyn_jacketSide_mstr_l(ComponentPose=@n26, Alpha=@n25) {
    alphaBoolBlend {blendOption=Linear}
    alphaScaleBiasClamp {bInterpResult=true}
    boneToModify {boneName=dyn_jacketSide_mstr_l}
    lookAtLocation {x=0, z=-11}
    lookAtTarget {boneReference {boneName=calf_l}}
    lookAt_Axis {axis=<1 0>}
    lookUp_Axis {axis {x=0}}
   }
   n28 = Get thigh_rot_up_l() -> (thigh_rot_up_l:double)
   n29 = "Transform (Modify) Bone - Bone: dyn_jacketFront_01_l"(ComponentPose=@n27, Alpha=@n28, Translation="0,0,0", Rotation="9.096,-4.987,-10.828", Scale="1,1,1") {
    rotationMode=BMM_Additive
    rotationSpace=BCS_BoneSpace
    translationMode=BMM_Additive
    translationSpace=BCS_BoneSpace
    alphaBoolBlend {blendOption=Linear}
    boneToModify {boneName=dyn_jacketFront_01_l}
    rotation=<rot 9.096 -4.987 -10.8284>
   }
   n30 = "Transform (Modify) Bone - Bone: dyn_jacketSide_01_l"(ComponentPose=@n29, Alpha=@n28, Translation="0,0,0", Rotation="5,0,0", Scale="1,1,1") {
    rotationMode=BMM_Additive
    rotationSpace=BCS_BoneSpace
    translationMode=BMM_Additive
    translationSpace=BCS_BoneSpace
    alphaBoolBlend {blendOption=Linear}
    boneToModify {boneName=dyn_jacketSide_01_l}
    rotation {pitch=5}
   }
   n31 = Get jumpingAlpha() -> (jumpingAlpha:double)
   n32 = "Transform (Modify) Bone - Bone: dyn_jacketBack_01_l"(ComponentPose=@n30, Alpha=@n31, Translation="0,0,0", Rotation="35,10,0.000", Scale="1,1,1") {
    rotationMode=BMM_Additive
    rotationSpace=BCS_BoneSpace
    alphaBoolBlend {blendOption=Linear}
    alphaScaleBiasClamp {bInterpResult=true, interpSpeedDecreasing=2, interpSpeedIncreasing=5}
    boneToModify {boneName=dyn_jacketBack_01_l}
    rotation {pitch=35, yaw=10}
   }
   n33 = "Transform (Modify) Bone - Bone: dyn_jacketBack_01_r"(ComponentPose=@n32, Alpha=@n31, Translation="0,0,0", Rotation="35,-10,0.000", Scale="1,1,1") {
    rotationMode=BMM_Additive
    rotationSpace=BCS_BoneSpace
    alphaBoolBlend {blendOption=Linear}
    alphaScaleBiasClamp {bInterpResult=true, interpSpeedDecreasing=2, interpSpeedIncreasing=5}
    boneToModify {boneName=dyn_jacketBack_01_r}
    rotation {pitch=35, yaw=-10}
   }
   n34 = "Transform (Modify) Bone - Bone: dyn_jacketSide_01_l"(ComponentPose=@n33, Alpha=@n31, Translation="0,0,0", Rotation="15,-15,0.000", Scale="1,1,1") {
    rotationMode=BMM_Additive
    rotationSpace=BCS_BoneSpace
    alphaBoolBlend {blendOption=Linear}
    alphaScaleBiasClamp {bInterpResult=true, interpSpeedDecreasing=2, interpSpeedIncreasing=5}
    boneToModify {boneName=dyn_jacketSide_01_l}
    rotation {pitch=15, yaw=-15}
   }
   n35 = "Transform (Modify) Bone - Bone: dyn_jacketSide_01_l"(ComponentPose=@n34, Alpha=@n31, Translation="0,0,0", Rotation="15,15,0", Scale="1,1,1") {
    rotationMode=BMM_Additive
    rotationSpace=BCS_BoneSpace
    alphaBoolBlend {blendOption=Linear}
    alphaScaleBiasClamp {bInterpResult=true, interpSpeedDecreasing=2, interpSpeedIncreasing=5}
    boneToModify {boneName=dyn_jacketSide_01_l}
    rotation {pitch=15, yaw=15}
   }
   n36 = Get skydiveAlpha() -> (skydiveAlpha:double)
   n37 = "Transform (Modify) Bone - Bone: dyn_jacketBack_mstr_l"(ComponentPose=@n35, Alpha=@n36, Translation="0,0,0", Rotation="15,0,0.000", Scale="1,1,1") {
    rotationMode=BMM_Additive
    rotationSpace=BCS_BoneSpace
    alphaBoolBlend {blendOption=Linear}
    alphaScaleBiasClamp {bInterpResult=true, interpSpeedDecreasing=2, interpSpeedIncreasing=5}
    boneToModify {boneName=dyn_jacketBack_mstr_l}
    rotation {pitch=15}
   }
   n38 = "Transform (Modify) Bone - Bone: dyn_jacketBack_03_l"(ComponentPose=@n37, Alpha=@n36, Translation="0,0,0", Rotation="20,0,0.000", Scale="1,1,1") {
    rotationMode=BMM_Additive
    rotationSpace=BCS_BoneSpace
    alphaBoolBlend {blendOption=Linear}
    alphaScaleBiasClamp {bInterpResult=true, interpSpeedDecreasing=2, interpSpeedIncreasing=5}
    boneToModify {boneName=dyn_jacketBack_03_l}
    rotation {pitch=20}
   }
   n39 = "Transform (Modify) Bone - Bone: dyn_jacketBack_mstr_r"(ComponentPose=@n38, Alpha=@n36, Translation="0,0,0", Rotation="15,0,0.000", Scale="1,1,1") {
    rotationMode=BMM_Additive
    rotationSpace=BCS_BoneSpace
    alphaBoolBlend {blendOption=Linear}
    alphaScaleBiasClamp {bInterpResult=true, interpSpeedDecreasing=2, interpSpeedIncreasing=5}
    boneToModify {boneName=dyn_jacketBack_mstr_r}
    rotation {pitch=15}
   }
   n40 = Get ParachuteAlpha() -> (ParachuteAlpha:double)
   n41 = "Transform (Modify) Bone - Bone: dyn_jacketBack_01_l"(ComponentPose=@n39, Alpha=@n40, Translation="0,0,0", Rotation="15,10,0.000", Scale="1,1,1") {
    rotationMode=BMM_Additive
    rotationSpace=BCS_BoneSpace
    alphaBoolBlend {blendOption=Linear}
    alphaScaleBiasClamp {bInterpResult=true, interpSpeedDecreasing=2, interpSpeedIncreasing=5}
    boneToModify {boneName=dyn_jacketBack_01_l}
    rotation {pitch=15, yaw=10}
   }
   n42 = "Transform (Modify) Bone - Bone: dyn_jacketBack_01_r"(ComponentPose=@n41, Alpha=@n40, Translation="0,0,0", Rotation="15,-10,0.000", Scale="1,1,1") {
    rotationMode=BMM_Additive
    rotationSpace=BCS_BoneSpace
    alphaBoolBlend {blendOption=Linear}
    alphaScaleBiasClamp {bInterpResult=true, interpSpeedDecreasing=2, interpSpeedIncreasing=5}
    boneToModify {boneName=dyn_jacketBack_01_r}
    rotation {pitch=15, yaw=-10}
   }
   n43 = "Transform (Modify) Bone - Bone: dyn_jacketSide_01_l"(ComponentPose=@n42, Alpha=@n40, Translation="0,0,0", Rotation="15,-15,0.000", Scale="1,1,1") {
    rotationMode=BMM_Additive
    rotationSpace=BCS_BoneSpace
    alphaBoolBlend {blendOption=Linear}
    alphaScaleBiasClamp {bInterpResult=true, interpSpeedDecreasing=2, interpSpeedIncreasing=5}
    boneToModify {boneName=dyn_jacketSide_01_l}
    rotation {pitch=15, yaw=-15}
   }
   n44 = "Transform (Modify) Bone - Bone: dyn_jacketSide_01_l"(ComponentPose=@n43, Alpha=@n40, Translation="0,0,0", Rotation="15,15,0.000", Scale="1,1,1") {
    rotationMode=BMM_Additive
    rotationSpace=BCS_BoneSpace
    alphaBoolBlend {blendOption=Linear}
    alphaScaleBiasClamp {bInterpResult=true, interpSpeedDecreasing=2, interpSpeedIncreasing=5}
    boneToModify {boneName=dyn_jacketSide_01_l}
    rotation {pitch=15, yaw=15}
   }
   n45 = Component To Local(ComponentPose=@n44)
   n46 = "Transform (Modify) Bone - Bone: dyn_jacketFront_01_r"(ComponentPose=Local To Component(LocalPose=Use cached pose 'prerigid'()), Alpha=1, Translation="0,0,0", Rotation="15,0,0", Scale="1,1,1") {
    rotationMode=BMM_Additive
    rotationSpace=BCS_BoneSpace
    translationMode=BMM_Additive
    translationSpace=BCS_BoneSpace
    alphaBoolBlend {blendOption=Linear}
    boneToModify {boneName=dyn_jacketFront_01_r}
    rotation {pitch=15}
   }
   n47 = "Transform (Modify) Bone - Bone: dyn_jacketSide_01_r"(ComponentPose=@n46, Alpha=1, Translation="0,0,0", Rotation="20,0,0", Scale="1,1,1") {
    rotationMode=BMM_Additive
    rotationSpace=BCS_BoneSpace
    translationMode=BMM_Additive
    translationSpace=BCS_BoneSpace
    alphaBoolBlend {blendOption=Linear}
    boneToModify {boneName=dyn_jacketSide_01_r}
    rotation {pitch=20}
   }
   n48 = "Transform (Modify) Bone - Bone: dyn_jacketBack_01_l"(ComponentPose=@n47, Alpha=@n2, Translation="0,0,0", Rotation="10,10,0.000", Scale="1,1,1") {
    rotationMode=BMM_Additive
    rotationSpace=BCS_BoneSpace
    alphaBoolBlend {blendOption=Linear}
    alphaScaleBiasClamp {bInterpResult=true, interpSpeedDecreasing=5}
    boneToModify {boneName=dyn_jacketBack_01_l}
    rotation {pitch=10, yaw=10}
   }
   n49 = "Transform (Modify) Bone - Bone: dyn_jacketBack_01_r"(ComponentPose=@n48, Alpha=@n2, Translation="0,0,0", Rotation="10,-10,0.000", Scale="1,1,1") {
    rotationMode=BMM_Additive
    rotationSpace=BCS_BoneSpace
    alphaBoolBlend {blendOption=Linear}
    alphaScaleBiasClamp {bInterpResult=true, interpSpeedDecreasing=5}
    boneToModify {boneName=dyn_jacketBack_01_r}
    rotation {pitch=10, yaw=-10}
   }
   n50 = "Transform (Modify) Bone - Bone: dyn_jacketBack_02_r"(ComponentPose=@n49, Alpha=@n2, Translation="0,0,0", Rotation="30,0,0.000", Scale="1,1,1") {
    rotationMode=BMM_Additive
    rotationSpace=BCS_BoneSpace
    alphaBoolBlend {blendOption=Linear}
    alphaScaleBiasClamp {bInterpResult=true, interpSpeedDecreasing=5}
    boneToModify {boneName=dyn_jacketBack_02_r}
    rotation {pitch=30}
   }
   n51 = Get thigh_rot_otr_r() -> (thigh_rot_otr_r:double)
   n52 = "Transform (Modify) Bone - Bone: dyn_jacketFront_01_r"(ComponentPose=@n50, Alpha=@n51, Translation="0,0,0", Rotation="45,0,0", Scale="1,1,1") {
    rotationMode=BMM_Additive
    rotationSpace=BCS_BoneSpace
    translationMode=BMM_Additive
    translationSpace=BCS_BoneSpace
    alphaBoolBlend {blendOption=Linear}
    boneToModify {boneName=dyn_jacketFront_01_r}
    rotation {pitch=45}
   }
   n53 = "Transform (Modify) Bone - Bone: dyn_jacketFront_01_r"(ComponentPose=@n52, Alpha=@n51, Translation="0,0,2", Rotation="30,0,0", Scale="1,1,1") {
    rotationMode=BMM_Additive
    rotationSpace=BCS_BoneSpace
    translationMode=BMM_Additive
    translationSpace=BCS_BoneSpace
    alphaBoolBlend {blendOption=Linear}
    boneToModify {boneName=dyn_jacketFront_01_r}
    rotation {pitch=30}
    translation {z=2}
   }
   n54 = "Transform (Modify) Bone - Bone: dyn_jacketSide_02_r"(ComponentPose=@n53, Alpha=@n51, Translation="0,0,0", Rotation="45,0,0", Scale="1,1,1") {
    rotationMode=BMM_Additive
    rotationSpace=BCS_BoneSpace
    translationMode=BMM_Additive
    translationSpace=BCS_BoneSpace
    alphaBoolBlend {blendOption=Linear}
    boneToModify {boneName=dyn_jacketSide_02_r}
    rotation {pitch=45}
   }
   n55 = Get thigh_rot_up_l()
   n56 = "Transform (Modify) Bone - Bone: dyn_jacketFront_01_l"(ComponentPose=@n54, Alpha=@n55, Translation="0,0,0", Rotation="10,-15.000,-35", Scale="1,1,1") {
    rotationMode=BMM_Additive
    rotationSpace=BCS_BoneSpace
    translationMode=BMM_Additive
    translationSpace=BCS_BoneSpace
    alphaBoolBlend {blendOption=Linear}
    boneToModify {boneName=dyn_jacketFront_01_l}
    rotation=<rot 10 -15 -35>
   }
   n57 = "Transform (Modify) Bone - Bone: dyn_jacketFront_02_l"(ComponentPose=@n56, Alpha=@n55, Translation="0,0,0", Rotation="-25,0,0", Scale="1,1,1") {
    rotationMode=BMM_Additive
    rotationSpace=BCS_BoneSpace
    translationMode=BMM_Additive
    translationSpace=BCS_BoneSpace
    alphaBoolBlend {blendOption=Linear}
    boneToModify {boneName=dyn_jacketFront_02_l}
    rotation {pitch=-25}
   }
   n58 = "Transform (Modify) Bone - Bone: dyn_jacketSide_01_l"(ComponentPose=@n57, Alpha=@n55, Translation="0,-2,3", Rotation="0,-10,0", Scale="1,1,1") {
    rotationMode=BMM_Additive
    rotationSpace=BCS_BoneSpace
    translationMode=BMM_Additive
    translationSpace=BCS_BoneSpace
    alphaBoolBlend {blendOption=Linear}
    boneToModify {boneName=dyn_jacketSide_01_l}
    rotation {yaw=-10}
    translation {y=-2, z=3}
   }
   n59 = Get thigh_rot_otr_l() -> (thigh_rot_otr_l:double)
   n60 = "Transform (Modify) Bone - Bone: dyn_jacketFront_01_l"(ComponentPose=@n58, Alpha=@n59, Translation="0,0,0", Rotation="45,0,0", Scale="1,1,1") {
    rotationMode=BMM_Additive
    rotationSpace=BCS_BoneSpace
    translationMode=BMM_Additive
    translationSpace=BCS_BoneSpace
    alphaBoolBlend {blendOption=Linear}
    boneToModify {boneName=dyn_jacketFront_01_l}
    rotation {pitch=45}
   }
   n61 = "Transform (Modify) Bone - Bone: dyn_jacketFront_02_l"(ComponentPose=@n60, Alpha=@n59, Translation="0,0,2", Rotation="30,0,0", Scale="1,1,1") {
    rotationMode=BMM_Additive
    rotationSpace=BCS_BoneSpace
    translationMode=BMM_Additive
    translationSpace=BCS_BoneSpace
    alphaBoolBlend {blendOption=Linear}
    boneToModify {boneName=dyn_jacketFront_02_l}
    rotation {pitch=30}
    translation {z=2}
   }
   n62 = "Transform (Modify) Bone - Bone: dyn_jacketSide_02_l"(ComponentPose=@n61, Alpha=@n59, Translation="0,0,0", Rotation="45,0,0", Scale="1,1,1") {
    rotationMode=BMM_Additive
    rotationSpace=BCS_BoneSpace
    translationMode=BMM_Additive
    translationSpace=BCS_BoneSpace
    alphaBoolBlend {blendOption=Linear}
    boneToModify {boneName=dyn_jacketSide_02_l}
    rotation {pitch=45}
   }
   n63 = Component To Local(ComponentPose=@n62)
   n64 = Get bIsCrouching()
   n65 = Get BwdMovement() -> (BwdMovement:double)
   n66 = Trail controller - Bone: dyn_jacketBack_05_r!(
    ComponentPose=Local To Component(LocalPose=Blend Poses by bool(BlendPose_0=@n63, BlendPose_1=@n45, bActiveValue=@n64) {blendType=Linear})
    Alpha=@n65
    FakeVelocity:Vector="0,0,0"
   ) -> (Pose) {
    bActorSpaceFakeVel=true
    bShowTrailLocation=true
    chainLength=4
    debugLifeTime=0.1
    maxDeltaTime=0.0333
    relaxationSpeedScale=3
    alphaScaleBiasClamp {bInterpResult=true}
    baseJoint {boneName=root}
    planarLimits [
     {drivingBone {boneName=root}, planeTransform {location=<0 0 10>, rotation=<rot 0 0 0>, scale=<1 1 1>}}
     {drivingBone {boneName=spine_03}, planeTransform {location=<0 -20 0>, rotation=<rot 0 15 -89.9999>, scale=<1 1 1>}}
    ]
    rotationLimits [{limitMax=<180 180 180>}, {limitMax=<180 180 180>}, {limitMax=<180 180 180>}, {limitMax=<180 180 180>}]
    rotationOffsets [<0 0 0>, <0 0 0>, <0 0 0>, <0 0 0>]
    trailBone {boneName=dyn_jacketBack_05_r}
    trailRelaxationSpeed {editorCurveData {keys [{arriveTangent=-72.0739, interpMode=RCIM_Cubic, leaveTangent=-72.0739, tangentMode=RCTM_User, time=0, value=35}, <at 1 5 RCIM_Cubic>]}}
   }
   n67 = Trail controller - Bone: dyn_jacketBack_05_l!(ComponentPose=@n66, Alpha=@n65, FakeVelocity:Vector="0,0,0") -> (Pose) {
    bActorSpaceFakeVel=true
    bShowTrailLocation=true
    chainLength=4
    debugLifeTime=0.1
    maxDeltaTime=0.0333
    relaxationSpeedScale=3
    alphaScaleBiasClamp {bInterpResult=true}
    baseJoint {boneName=root}
    planarLimits [
     {drivingBone {boneName=root}, planeTransform {location=<0 0 10>, rotation=<rot 0 0 0>, scale=<1 1 1>}}
     {drivingBone {boneName=spine_03}, planeTransform {location=<0 -20 0>, rotation=<rot 0 15 -89.9999>, scale=<1 1 1>}}
    ]
    rotationLimits [{limitMax=<180 180 180>}, {limitMax=<180 180 180>}, {limitMax=<180 180 180>}, {limitMax=<180 180 180>}]
    rotationOffsets [<0 0 0>, <0 0 0>, <0 0 0>, <0 0 0>]
    trailBone {boneName=dyn_jacketBack_05_l}
    trailRelaxationSpeed {editorCurveData {keys [{arriveTangent=-72.0739, interpMode=RCIM_Cubic, leaveTangent=-72.0739, tangentMode=RCTM_User, time=0, value=35}, <at 1 5 RCIM_Cubic>]}}
   }
   n68 = Get LeftMovement() -> (LeftMovement:double)
   n69 = Trail controller - Bone: dyn_jacketFront_05_r!(ComponentPose=@n67, Alpha=@n68, FakeVelocity:Vector="0,0,0") -> (Pose) {
    bActorSpaceFakeVel=true
    bShowTrailLocation=true
    chainLength=4
    debugLifeTime=0.1
    maxDeltaTime=0.0333
    relaxationSpeedScale=3
    alphaScaleBiasClamp {bInterpResult=true}
    baseJoint {boneName=root}
    planarLimits [
     {drivingBone {boneName=root}, planeTransform {location=<0 0 10>, rotation=<rot 0 0 0>, scale=<1 1 1>}}
     {drivingBone {boneName=spine_03}, planeTransform {location=<0 -20 0>, rotation=<rot 0 15 -89.9999>, scale=<1 1 1>}}
    ]
    rotationLimits [{limitMax=<180 180 180>}, {limitMax=<180 180 180>}, {limitMax=<180 180 180>}, {limitMax=<180 180 180>}]
    rotationOffsets [<0 0 0>, <0 0 0>, <0 0 0>, <0 0 0>]
    trailBone {boneName=dyn_jacketFront_05_r}
    trailRelaxationSpeed {editorCurveData {keys [{arriveTangent=-72.0739, interpMode=RCIM_Cubic, leaveTangent=-72.0739, tangentMode=RCTM_User, time=0, value=35}, <at 1 5 RCIM_Cubic>]}}
   }
   n70 = Trail controller - Bone: dyn_jacketSide_05_r!(ComponentPose=@n69, Alpha=@n68, FakeVelocity:Vector="0,0,0") -> (Pose) {
    bActorSpaceFakeVel=true
    bShowTrailLocation=true
    chainLength=4
    debugLifeTime=0.1
    maxDeltaTime=0.0333
    relaxationSpeedScale=3
    alphaScaleBiasClamp {bInterpResult=true}
    baseJoint {boneName=root}
    planarLimits [
     {drivingBone {boneName=root}, planeTransform {location=<0 0 10>, rotation=<rot 0 0 0>, scale=<1 1 1>}}
     {drivingBone {boneName=spine_03}, planeTransform {location=<0 -20 0>, rotation=<rot 0 15 -89.9999>, scale=<1 1 1>}}
    ]
    rotationLimits [{limitMax=<180 180 180>}, {limitMax=<180 180 180>}, {limitMax=<180 180 180>}, {limitMax=<180 180 180>}]
    rotationOffsets [<0 0 0>, <0 0 0>, <0 0 0>, <0 0 0>]
    trailBone {boneName=dyn_jacketSide_05_r}
    trailRelaxationSpeed {editorCurveData {keys [{arriveTangent=-72.0739, interpMode=RCIM_Cubic, leaveTangent=-72.0739, tangentMode=RCTM_User, time=0, value=35}, <at 1 5 RCIM_Cubic>]}}
   }
   n71 = Get RightMovement() -> (RightMovement:double)
   n72 = Trail controller - Bone: dyn_jacketFront_05_l!(ComponentPose=@n70, Alpha=@n71, FakeVelocity:Vector="0,0,0") -> (Pose) {
    bActorSpaceFakeVel=true
    bShowTrailLocation=true
    chainLength=4
    debugLifeTime=0.1
    maxDeltaTime=0.0333
    relaxationSpeedScale=3
    alphaScaleBiasClamp {bInterpResult=true}
    baseJoint {boneName=root}
    planarLimits [
     {drivingBone {boneName=root}, planeTransform {location=<0 0 10>, rotation=<rot 0 0 0>, scale=<1 1 1>}}
     {drivingBone {boneName=spine_03}, planeTransform {location=<0 -20 0>, rotation=<rot 0 15 -89.9999>, scale=<1 1 1>}}
    ]
    rotationLimits [{limitMax=<180 180 180>}, {limitMax=<180 180 180>}, {limitMax=<180 180 180>}, {limitMax=<180 180 180>}]
    rotationOffsets [<0 0 0>, <0 0 0>, <0 0 0>, <0 0 0>]
    trailBone {boneName=dyn_jacketFront_05_l}
    trailRelaxationSpeed {editorCurveData {keys [{arriveTangent=-72.0739, interpMode=RCIM_Cubic, leaveTangent=-72.0739, tangentMode=RCTM_User, time=0, value=35}, <at 1 5 RCIM_Cubic>]}}
   }
   n73 = Trail controller - Bone: dyn_jacketSide_05_l!(ComponentPose=@n72, Alpha=@n71, FakeVelocity:Vector="0,0,0") -> (Pose) {
    bActorSpaceFakeVel=true
    bShowTrailLocation=true
    chainLength=4
    debugLifeTime=0.1
    maxDeltaTime=0.0333
    relaxationSpeedScale=3
    alphaScaleBiasClamp {bInterpResult=true}
    baseJoint {boneName=root}
    planarLimits [
     {drivingBone {boneName=root}, planeTransform {location=<0 0 10>, rotation=<rot 0 0 0>, scale=<1 1 1>}}
     {drivingBone {boneName=spine_03}, planeTransform {location=<0 -20 0>, rotation=<rot 0 15 -89.9999>, scale=<1 1 1>}}
    ]
    rotationLimits [{limitMax=<180 180 180>}, {limitMax=<180 180 180>}, {limitMax=<180 180 180>}, {limitMax=<180 180 180>}]
    rotationOffsets [<0 0 0>, <0 0 0>, <0 0 0>, <0 0 0>]
    trailBone {boneName=dyn_jacketSide_05_l}
    trailRelaxationSpeed {editorCurveData {keys [{arriveTangent=-72.0739, interpMode=RCIM_Cubic, leaveTangent=-72.0739, tangentMode=RCTM_User, time=0, value=35}, <at 1 5 RCIM_Cubic>]}}
   }
   n74 = Component To Local(ComponentPose=@n73)
   n75 = Get bRigidBodyEnabled()
   n76 = Anim Dynamics - Bone: dyn_spur_l(
    ComponentPose=Local To Component(LocalPose=Blend Poses by bool(BlendPose_0=@n18, BlendPose_1=@n74, bActiveValue=@n75) {blendType=Linear})
    Alpha=1
    ExternalForce="0,0,0"
   ) {
    angularSpringConstant=800
    bAngularSpring=true
    bGravityOverrideInSimSpace=true
    bOverrideAngularDamping=true
    bUseGravityOverride=true
    gravityScale=0.2
    simulationSpace=RootRelative
    alphaBoolBlend {blendOption=Linear}
    boundBone {boneName=dyn_spur_l}
    componentAppliedLinearAccClamp=<3000 3000 3000>
    componentLinearAccScale=<0.2 0.2 0.25>
    componentLinearVelScale=<0.45 0.45 0.45>
    gravityOverride {z=-800}
    physicsBodyDefinitions [{
      sphereCollisionRadius=0
      boundBone {boneName=dyn_spur_l}
      boxExtents=<20 20 10>
      constraintSetup {angularTargetAxis=AxisY, twistAxis=AxisZ, angularLimitsMax=<180 0 0>, angularLimitsMin=<-180 0 0>, angularTarget=<0 0 0>}
      localJointOffset=<0.01 -0.01 -10>
     }]
    relativeSpaceBone {boneName=spine_03}
   }
   n77 = Anim Dynamics - Bone: dyn_spur_r(ComponentPose=@n76, Alpha=1, ExternalForce="0,0,0") {
    angularSpringConstant=800
    bAngularSpring=true
    bGravityOverrideInSimSpace=true
    bOverrideAngularDamping=true
    bUseGravityOverride=true
    gravityScale=0.2
    simulationSpace=RootRelative
    alphaBoolBlend {blendOption=Linear}
    boundBone {boneName=dyn_spur_r}
    componentAppliedLinearAccClamp=<3000 3000 3000>
    componentLinearAccScale=<0.2 0.2 0.25>
    componentLinearVelScale=<0.45 0.45 0.45>
    gravityOverride {z=-800}
    physicsBodyDefinitions [{
      sphereCollisionRadius=0
      boundBone {boneName=dyn_spur_r}
      boxExtents=<20 20 10>
      constraintSetup {angularTargetAxis=AxisY, twistAxis=AxisZ, angularLimitsMax=<180 0 0>, angularLimitsMin=<-180 0 0>, angularTarget=<0 0 0>}
      localJointOffset=<0.01 0.01 10>
     }]
    relativeSpaceBone {boneName=pelvis}
   }
   n78 = "Constraint - dyn_tassle_forearm_l (PPPP)"!(
    ComponentPose=@n77
    Alpha=1
    ConstraintWeights_0=0.897
    ConstraintWeights_1:float=0.0641
    ConstraintWeights_2:float=0.0246
    ConstraintWeights_3:float=0.0143
   ) -> (Pose) {
    alphaBoolBlend {blendOption=Linear}
    boneToModify {boneName=dyn_tassle_forearm_l}
    constraintSetup [
     {targetBone {boneName=lowerarm_twist_01_l}}
     {targetBone {boneName=hand_l}}
     {targetBone {boneName=lowerarm_l}}
     {targetBone {boneName=lowerarm_twist_02_l}}
    ]
    constraintWeights [1,1,1,1]
   }
   n79 = "Constraint - dyn_tassle_forearm_r (PPPP)"!(
    ComponentPose=@n78
    Alpha=1
    ConstraintWeights_0=0.897
    ConstraintWeights_1:float=0.0641
    ConstraintWeights_2:float=0.0246
    ConstraintWeights_3:float=0.0143
   ) -> (Pose) {
    alphaBoolBlend {blendOption=Linear}
    boneToModify {boneName=dyn_tassle_forearm_r}
    constraintSetup [
     {targetBone {boneName=lowerarm_twist_01_r}}
     {targetBone {boneName=hand_r}}
     {targetBone {boneName=lowerarm_r}}
     {targetBone {boneName=lowerarm_twist_02_r}}
    ]
    constraintWeights [1,1,1,1]
   }
   n80 = Anim Dynamics - Bone: dyn_tassle_forearm_l(ComponentPose=@n79, Alpha=1, ExternalForce="0,0,0") {
    angularDampingOverride=0.99
    angularSpringConstant=3000
    bAngularSpring=true
    bGravityOverrideInSimSpace=true
    bOverrideAngularDamping=true
    bUseGravityOverride=true
    gravityScale=0.35
    simulationSpace=RootRelative
    alphaBoolBlend {blendOption=Linear}
    boundBone {boneName=dyn_tassle_forearm_l}
    componentAppliedLinearAccClamp=<3000 3000 3000>
    componentLinearAccScale=<0.4 0.4 0.4>
    componentLinearVelScale=<0.4 0.4 0.4>
    gravityOverride {z=-980}
    physicsBodyDefinitions [{
      sphereCollisionRadius=0
      boundBone {boneName=dyn_tassle_forearm_l}
      boxExtents=<25 15 19>
      constraintSetup {angularTargetAxis=AxisY, twistAxis=AxisY, angularLimitsMax=<89 10 10>, angularLimitsMin=<-89 -10 -10>, angularTarget=<0 1 0>}
      localJointOffset=<0 -4.5 0>
     }]
   }
   n81 = Anim Dynamics - Bone: dyn_tassle_forearm_r(ComponentPose=@n80, Alpha=1, ExternalForce="0,0,0") {
    angularDampingOverride=0.99
    angularSpringConstant=3000
    bAngularSpring=true
    bGravityOverrideInSimSpace=true
    bOverrideAngularDamping=true
    bUseGravityOverride=true
    gravityScale=0.35
    simulationSpace=RootRelative
    alphaBoolBlend {blendOption=Linear}
    boundBone {boneName=dyn_tassle_forearm_r}
    componentAppliedLinearAccClamp=<3000 3000 3000>
    componentLinearAccScale=<0.4 0.4 0.4>
    componentLinearVelScale=<0.4 0.4 0.4>
    gravityOverride {z=-980}
    physicsBodyDefinitions [{
      sphereCollisionRadius=0
      boundBone {boneName=dyn_tassle_forearm_r}
      boxExtents=<28 13 21>
      constraintSetup {angularTargetAxis=AxisY, twistAxis=AxisY, angularLimitsMax=<89 10 10>, angularLimitsMin=<-89 -10 -10>, angularTarget=<0 1 0>}
      localJointOffset=<0 4.5 0>
     }]
   }
   n82 = Anim Dynamics - Bone: dyn_scarf_chest_bk_02(ComponentPose=@n81, Alpha=1, ExternalForce="0,0,0") {
    angularDampingOverride=0.9999
    angularSpringConstant=100
    bAngularSpring=true
    bGravityOverrideInSimSpace=true
    bOverrideAngularDamping=true
    simulationSpace=RootRelative
    alphaBoolBlend {blendOption=Linear}
    boundBone {boneName=dyn_scarf_chest_bk_02}
    componentAppliedLinearAccClamp=<3500 3500 3500>
    componentLinearAccScale=<1.2 1.2 1.2>
    componentLinearVelScale=<1.2 1.2 1.2>
    physicsBodyDefinitions [{
      sphereCollisionRadius=0
      boundBone {boneName=dyn_scarf_chest_bk_02}
      boxExtents=<10 20 10>
      constraintSetup {angularLimitsMax=<1 5 15>, angularLimitsMin=<-1 -30 -15>, angularTarget=<0 0 0>}
      localJointOffset=<2 0 0>
     }]
    relativeSpaceBone {boneName=dyn_root_simspace}
   }
   n83 = Anim Dynamics - Bone: dyn_scarf_arm_03(ComponentPose=@n82, Alpha=1, ExternalForce="0,0,0") {
    angularDampingOverride=0.9999
    angularSpringConstant=100
    bAngularSpring=true
    bGravityOverrideInSimSpace=true
    bOverrideAngularDamping=true
    simulationSpace=RootRelative
    alphaBoolBlend {blendOption=Linear}
    boundBone {boneName=dyn_scarf_arm_03}
    componentAppliedLinearAccClamp=<3500 3500 3500>
    componentLinearAccScale=<1.2 1.2 1.2>
    componentLinearVelScale=<1.2 1.2 1.2>
    physicsBodyDefinitions [{
      sphereCollisionRadius=0
      boundBone {boneName=dyn_scarf_arm_03}
      boxExtents=<10 20 10>
      constraintSetup {angularLimitsMax=<1 0 15>, angularLimitsMin=<-1 -30 -15>, angularTarget=<0 0 0>}
      localJointOffset=<2 0 0>
     }]
    relativeSpaceBone {boneName=dyn_root_simspace}
   }
   n84 = Anim Dynamics - Bone: dyn_scarf_arm_bk_03(ComponentPose=@n83, Alpha=1, ExternalForce="0,0,0") {
    angularDampingOverride=0.9999
    angularSpringConstant=100
    bAngularSpring=true
    bGravityOverrideInSimSpace=true
    bOverrideAngularDamping=true
    simulationSpace=RootRelative
    alphaBoolBlend {blendOption=Linear}
    boundBone {boneName=dyn_scarf_arm_bk_03}
    componentAppliedLinearAccClamp=<3500 3500 3500>
    componentLinearAccScale=<1.2 1.2 1.2>
    componentLinearVelScale=<1.2 1.2 1.2>
    physicsBodyDefinitions [{
      sphereCollisionRadius=0
      boundBone {boneName=dyn_scarf_arm_bk_03}
      boxExtents=<10 20 10>
      constraintSetup {angularLimitsMax=<1 5 15>, angularLimitsMin=<-1 -30 -15>, angularTarget=<0 0 0>}
      localJointOffset=<2 0 0>
     }]
    relativeSpaceBone {boneName=dyn_root_simspace}
   }
   n85 = Anim Dynamics - Bone: dyn_scarf_arm_fr_02(ComponentPose=@n84, Alpha=1, ExternalForce="0,0,0") {
    angularDampingOverride=0.9999
    angularSpringConstant=100
    bAngularSpring=true
    bGravityOverrideInSimSpace=true
    bOverrideAngularDamping=true
    simulationSpace=RootRelative
    alphaBoolBlend {blendOption=Linear}
    boundBone {boneName=dyn_scarf_arm_fr_02}
    componentAppliedLinearAccClamp=<3500 3500 3500>
    componentLinearAccScale=<1.2 1.2 1.2>
    componentLinearVelScale=<1.2 1.2 1.2>
    physicsBodyDefinitions [{
      sphereCollisionRadius=0
      boundBone {boneName=dyn_scarf_arm_fr_02}
      boxExtents=<10 20 10>
      constraintSetup {angularLimitsMax=<1 0 15>, angularLimitsMin=<-1 -30 -15>, angularTarget=<0 0 0>}
      localJointOffset=<2 0 0>
     }]
    relativeSpaceBone {boneName=dyn_root_simspace}
   }
   n86 = Anim Dynamics - Bone: dyn_scarf_chest_l_01(ComponentPose=@n85, Alpha=1, ExternalForce="0,0,0") {
    angularDampingOverride=0.9999
    angularSpringConstant=100
    bAngularSpring=true
    bGravityOverrideInSimSpace=true
    bOverrideAngularDamping=true
    simulationSpace=RootRelative
    alphaBoolBlend {blendOption=Linear}
    boundBone {boneName=dyn_scarf_chest_l_01}
    componentAppliedLinearAccClamp=<3500 3500 3500>
    componentLinearAccScale=<1.2 1.2 1.2>
    componentLinearVelScale=<1.2 1.2 1.2>
    physicsBodyDefinitions [{
      sphereCollisionRadius=0
      boundBone {boneName=dyn_scarf_chest_l_01}
      boxExtents=<10 20 10>
      constraintSetup {angularLimitsMax=<1 0 15>, angularLimitsMin=<-1 -30 -15>, angularTarget=<0 0 0>}
      localJointOffset=<2 0 0>
     }]
    relativeSpaceBone {boneName=dyn_root_simspace}
   }
   n87 = Anim Dynamics - Bone: dyn_scarf_chest_bk_tassels(ComponentPose=@n86, Alpha=1, ExternalForce="0,0,0") {
    angularDampingOverride=0.9999
    angularSpringConstant=100
    bAngularSpring=true
    bGravityOverrideInSimSpace=true
    bOverrideAngularDamping=true
    simulationSpace=RootRelative
    alphaBoolBlend {blendOption=Linear}
    boundBone {boneName=dyn_scarf_chest_bk_tassels}
    componentAppliedLinearAccClamp=<3500 3500 3500>
    componentLinearAccScale=<1.2 1.2 1.2>
    componentLinearVelScale=<1.2 1.2 1.2>
    physicsBodyDefinitions [{
      sphereCollisionRadius=0
      boundBone {boneName=dyn_scarf_chest_bk_tassels}
      boxExtents=<10 20 10>
      constraintSetup {angularLimitsMax=<1 5 15>, angularLimitsMin=<-1 -30 -15>, angularTarget=<0 0 0>}
      localJointOffset=<2 0 0>
     }]
    relativeSpaceBone {boneName=dyn_root_simspace}
   }
   n88 = Anim Dynamics - Bone: dyn_scarf_arm_tassels(ComponentPose=@n87, Alpha=1, ExternalForce="0,0,0") {
    angularDampingOverride=0.9999
    angularSpringConstant=100
    bAngularSpring=true
    bGravityOverrideInSimSpace=true
    bOverrideAngularDamping=true
    simulationSpace=RootRelative
    alphaBoolBlend {blendOption=Linear}
    boundBone {boneName=dyn_scarf_arm_tassels}
    componentAppliedLinearAccClamp=<3500 3500 3500>
    componentLinearAccScale=<1.2 1.2 1.2>
    componentLinearVelScale=<1.2 1.2 1.2>
    physicsBodyDefinitions [{
      sphereCollisionRadius=0
      boundBone {boneName=dyn_scarf_arm_tassels}
      boxExtents=<10 20 10>
      constraintSetup {angularLimitsMax=<1 0 15>, angularLimitsMin=<-1 -30 -15>, angularTarget=<0 0 0>}
      localJointOffset=<2 0 0>
     }]
    relativeSpaceBone {boneName=dyn_root_simspace}
   }
   n89 = Anim Dynamics - Bone: dyn_scarf_arm_bk_tassels(ComponentPose=@n88, Alpha=1, ExternalForce="0,0,0") {
    angularDampingOverride=0.9999
    angularSpringConstant=100
    bAngularSpring=true
    bGravityOverrideInSimSpace=true
    bOverrideAngularDamping=true
    simulationSpace=RootRelative
    alphaBoolBlend {blendOption=Linear}
    boundBone {boneName=dyn_scarf_arm_bk_tassels}
    componentAppliedLinearAccClamp=<3500 3500 3500>
    componentLinearAccScale=<1.2 1.2 1.2>
    componentLinearVelScale=<1.2 1.2 1.2>
    physicsBodyDefinitions [{
      sphereCollisionRadius=0
      boundBone {boneName=dyn_scarf_arm_bk_tassels}
      boxExtents=<10 20 10>
      constraintSetup {angularLimitsMax=<1 5 15>, angularLimitsMin=<-1 -30 -15>, angularTarget=<0 0 0>}
      localJointOffset=<2 0 0>
     }]
    relativeSpaceBone {boneName=dyn_root_simspace}
   }
   n90 = Anim Dynamics - Bone: dyn_scarf_arm_fr_03(ComponentPose=@n89, Alpha=1, ExternalForce="0,0,0") {
    angularDampingOverride=0.9999
    angularSpringConstant=100
    bAngularSpring=true
    bGravityOverrideInSimSpace=true
    bOverrideAngularDamping=true
    simulationSpace=RootRelative
    alphaBoolBlend {blendOption=Linear}
    boundBone {boneName=dyn_scarf_arm_fr_03}
    componentAppliedLinearAccClamp=<3500 3500 3500>
    componentLinearAccScale=<1.2 1.2 1.2>
    componentLinearVelScale=<1.2 1.2 1.2>
    physicsBodyDefinitions [{
      sphereCollisionRadius=0
      boundBone {boneName=dyn_scarf_arm_fr_03}
      boxExtents=<10 20 10>
      constraintSetup {angularLimitsMax=<1 0 15>, angularLimitsMin=<-1 -30 -15>, angularTarget=<0 0 0>}
      localJointOffset=<2 0 0>
     }]
    relativeSpaceBone {boneName=dyn_root_simspace}
   }
   n91 = Anim Dynamics - Bone: dyn_scarf_chest_l_02(ComponentPose=@n90, Alpha=1, ExternalForce="0,0,0") {
    angularDampingOverride=0.9999
    angularSpringConstant=100
    bAngularSpring=true
    bGravityOverrideInSimSpace=true
    bOverrideAngularDamping=true
    simulationSpace=RootRelative
    alphaBoolBlend {blendOption=Linear}
    boundBone {boneName=dyn_scarf_chest_l_02}
    componentAppliedLinearAccClamp=<3500 3500 3500>
    componentLinearAccScale=<1.2 1.2 1.2>
    componentLinearVelScale=<1.2 1.2 1.2>
    physicsBodyDefinitions [{
      sphereCollisionRadius=0
      boundBone {boneName=dyn_scarf_chest_l_02}
      boxExtents=<10 20 10>
      constraintSetup {angularLimitsMax=<1 0 15>, angularLimitsMin=<-1 -30 -15>, angularTarget=<0 0 0>}
      localJointOffset=<2 0 0>
     }]
    relativeSpaceBone {boneName=dyn_root_simspace}
   }
   Output Pose(Result=Component To Local(ComponentPose=@n91)) {layerGroup=None}
   n92 = "Transform (Modify) Bone - Bone: dyn_glute_l"(
    ComponentPose=Local To Component(LocalPose=Copy Pose From Mesh() {bUseAttachedParent=true})
    Alpha=Get thigh_rot_l()
    Translation="-4.689,5.014,-4.028"
    Rotation="0.000,0.000,-55"
    Scale="1.25,1.25,1.25"
   ) {
    rotationMode=BMM_Additive
    rotationSpace=BCS_BoneSpace
    scaleMode=BMM_Additive
    scaleSpace=BCS_BoneSpace
    translationMode=BMM_Additive
    translationSpace=BCS_BoneSpace
    alphaBoolBlend {blendOption=Linear}
    boneToModify {boneName=dyn_glute_l}
    rotation {roll=-55}
    scale=<1.25 1.25 1.25>
    translation=<-4.689 5.014 -4.028>
   }
   n93 = "Transform (Modify) Bone - Bone: dyn_glute_r"(ComponentPose=@n92, Alpha=Get thigh_rot_r(), Translation="-3.75,0,0", Rotation="0.000,0.000,-55", Scale="1.3,1.3,1.3") {
    rotationMode=BMM_Additive
    rotationSpace=BCS_BoneSpace
    scaleMode=BMM_Additive
    scaleSpace=BCS_BoneSpace
    translationMode=BMM_Additive
    translationSpace=BCS_BoneSpace
    alphaBoolBlend {blendOption=Linear}
    boneToModify {boneName=dyn_glute_r}
    rotation {roll=-55}
    scale=<1.3 1.3 1.3>
    translation {x=-3.75}
   }
   n94 = "Constraint - dyn_jacketFront_mstr_l (RR)"!(ComponentPose=@n93, Alpha=1, ConstraintWeights_0=0.8, ConstraintWeights_1:float=0.2) -> (Pose) {
    alphaBoolBlend {blendOption=Linear}
    boneToModify {boneName=dyn_jacketFront_mstr_l}
    constraintSetup [{transformType=Rotation, targetBone {boneName=thigh_twist_01_l}}, {transformType=Rotation, targetBone {boneName=pelvis}}]
    constraintWeights [1,1]
   }
   n95 = "Transform (Modify) Bone - Bone: dyn_jacketFront_mstr_l"(ComponentPose=@n94, Alpha=Get thigh_rot_l(), Translation="-7,2,10", Rotation="5,-10,-5", Scale="1,1,1") {
    rotationMode=BMM_Additive
    rotationSpace=BCS_BoneSpace
    scaleMode=BMM_Additive
    scaleSpace=BCS_BoneSpace
    translationMode=BMM_Additive
    translationSpace=BCS_BoneSpace
    alphaBoolBlend {blendOption=Linear}
    boneToModify {boneName=dyn_jacketFront_mstr_l}
    rotation=<rot 5 -10 -5>
    translation=<-7 2 10>
   }
   n96 = "Constraint - dyn_jacketFront_mstr_r (RR)"!(ComponentPose=@n95, Alpha=1, ConstraintWeights_0=0.7, ConstraintWeights_1:float=0.3) -> (Pose) {
    alphaBoolBlend {blendOption=Linear}
    boneToModify {boneName=dyn_jacketFront_mstr_r}
    constraintSetup [{transformType=Rotation, targetBone {boneName=thigh_r}}, {transformType=Rotation, targetBone {boneName=pelvis}}]
    constraintWeights [1,1]
   }
   n97 = "Transform (Modify) Bone - Bone: dyn_jacketFront_mstr_r"(ComponentPose=@n96, Alpha=Get thigh_rot_r(), Translation="-15,0,5", Rotation="25,0,15", Scale="1,1,1") {
    rotationMode=BMM_Additive
    rotationSpace=BCS_BoneSpace
    scaleMode=BMM_Additive
    scaleSpace=BCS_BoneSpace
    translationMode=BMM_Additive
    translationSpace=BCS_BoneSpace
    alphaBoolBlend {blendOption=Linear}
    boneToModify {boneName=dyn_jacketFront_mstr_r}
    rotation {pitch=25, roll=15}
    translation {x=-15, z=5}
   }
   n98 = "Constraint - dyn_jacketSide_01_r (RR)"!(ComponentPose=@n97, Alpha=1, ConstraintWeights_0=0.1, ConstraintWeights_1:float=0.9) -> (Pose) {
    alphaBoolBlend {blendOption=Linear}
    boneToModify {boneName=dyn_jacketSide_01_r}
    constraintSetup [{transformType=Rotation, targetBone {boneName=dyn_jacketFront_01_r}}, {transformType=Rotation, targetBone {boneName=pelvis}}]
    constraintWeights [1,1]
   }
   n99 = "Constraint - dyn_jacketSide_01_l (RR)"!(ComponentPose=@n98, Alpha=1, ConstraintWeights_0=0.1, ConstraintWeights_1:float=0.9) -> (Pose) {
    alphaBoolBlend {blendOption=Linear}
    boneToModify {boneName=dyn_jacketSide_01_l}
    constraintSetup [{transformType=Rotation, targetBone {boneName=dyn_jacketFront_01_l}}, {transformType=Rotation, targetBone {boneName=pelvis}}]
    constraintWeights [1,1]
   }
   n100 = "Constraint - dyn_scarf_arm_01 (PP)"(ComponentPose=@n99, Alpha=1, ConstraintWeights_0=0.7179, ConstraintWeights_1=0.2821) {
    alphaBoolBlend {blendOption=Linear}
    boneToModify {boneName=dyn_scarf_arm_01}
    constraintSetup [{targetBone {boneName=upperarm_twist_01_l}}, {targetBone {boneName=clavicle_l}}]
    constraintWeights [1,1]
   }
   n101 = "Constraint - dyn_scarf_arm_bk_01 (PP)"!(ComponentPose=@n100, Alpha=1, ConstraintWeights_0=0.2441, ConstraintWeights_1:float=0.7559) -> (Pose) {
    alphaBoolBlend {blendOption=Linear}
    boneToModify {boneName=dyn_scarf_arm_bk_01}
    constraintSetup [{targetBone {boneName=upperarm_twist_01_l}}, {targetBone {boneName=clavicle_l}}]
    constraintWeights [1,1]
   }
   n102 = "Constraint - dyn_scarf_chest_l_01 (RR)"!(ComponentPose=@n101, Alpha=1, ConstraintWeights_0=0.25, ConstraintWeights_1:float=0.75) -> (Pose) {
    alphaBoolBlend {blendOption=Linear}
    boneToModify {boneName=dyn_scarf_chest_l_01}
    constraintSetup [{transformType=Rotation, targetBone {boneName=dyn_scarf_arm_fr_01}}, {transformType=Rotation, targetBone {boneName=dyn_scarf_chest_r_01}}]
    constraintWeights [1,1]
   }
   n103 = "Constraint - dyn_scarf_arm_fr_01 (PPP)"!(ComponentPose=@n102, Alpha=0.6, ConstraintWeights_0=0.9, ConstraintWeights_1:float=0.1, ConstraintWeights_2:float=0) -> (Pose) {
    alphaBoolBlend {blendOption=Linear}
    boneToModify {boneName=dyn_scarf_arm_fr_01}
    constraintSetup [{targetBone {boneName=upperarm_twist_01_l}}, {targetBone {boneName=clavicle_l}}, {targetBone {boneName=spine_05}}]
    constraintWeights [1,1,1]
   }
   n104 = "Transform (Modify) Bone - Bone: dyn_scarf_arm_01"!(ComponentPose=@n103, bAlphaBoolEnabled:bool=@n1, Translation="0,0,0", Rotation="10,-10,0", Scale="1,1,1") -> (Pose) {
    alphaInputType=Bool
    rotationMode=BMM_Additive
    rotationSpace=BCS_BoneSpace
    alphaBoolBlend {blendInTime=0.1, blendOption=Linear, blendOutTime=0.5}
    alphaScaleBiasClamp {bInterpResult=true}
    boneToModify {boneName=dyn_scarf_arm_01}
    rotation {pitch=10, yaw=-10}
   }
   n105 = "Transform (Modify) Bone - Bone: dyn_scarf_arm_bk_01"!(ComponentPose=@n104, bAlphaBoolEnabled:bool=@n1, Translation="0,0,0", Rotation="15,0,0", Scale="1,1,1") -> (Pose) {
    alphaInputType=Bool
    rotationMode=BMM_Additive
    rotationSpace=BCS_BoneSpace
    alphaBoolBlend {blendInTime=0.1, blendOption=Linear, blendOutTime=0.5}
    alphaScaleBiasClamp {bInterpResult=true}
    boneToModify {boneName=dyn_scarf_arm_bk_01}
    rotation {pitch=15}
   }
   n106 = "Transform (Modify) Bone - Bone: dyn_scarf_chest_bk_01"!(ComponentPose=@n105, bAlphaBoolEnabled:bool=@n1, Translation="0,0,0", Rotation="15,0,0", Scale="1,1,1") -> (Pose) {
    alphaInputType=Bool
    rotationMode=BMM_Additive
    rotationSpace=BCS_BoneSpace
    alphaBoolBlend {blendInTime=0.1, blendOption=Linear, blendOutTime=0.5}
    alphaScaleBiasClamp {bInterpResult=true}
    boneToModify {boneName=dyn_scarf_chest_bk_01}
    rotation {pitch=15}
   }
   n107 = Get skydiveAlpha()
   n108 = "Transform (Modify) Bone - Bone: dyn_scarf_arm_01"!(ComponentPose=@n106, Alpha:float=@n107, Translation="0,0,0", Rotation="15,-5,0", Scale="1,1,1") -> (Pose) {
    rotationMode=BMM_Additive
    rotationSpace=BCS_BoneSpace
    alphaBoolBlend {blendOption=Linear}
    boneToModify {boneName=dyn_scarf_arm_01}
    rotation {pitch=15, yaw=-5}
   }
   n109 = "Transform (Modify) Bone - Bone: dyn_scarf_arm_bk_01"!(ComponentPose=@n108, Alpha:float=@n107, Translation="0,0,0", Rotation="15,0.000,0.000", Scale="1,1,1") -> (Pose) {
    rotationMode=BMM_Additive
    rotationSpace=BCS_BoneSpace
    alphaBoolBlend {blendOption=Linear}
    boneToModify {boneName=dyn_scarf_arm_bk_01}
    rotation {pitch=15}
   }
   Save cached pose 'prerigid'(Pose=Component To Local(ComponentPose=@n109)) {cacheName=prerigid}
  ]}
 EventGraph {nodes [
   Event BlueprintUpdateAnimation()
   > Get!(execute:Exec) -> (then:Exec, else:Exec, Skel_Mesh:Skeletal Mesh Component Object Reference)
   > n1 = Sequence!(execute) -> (then_0, then_1, then_2:Exec, then_3:Exec, then_4:Exec, then_5:Exec, then_6:Exec, then_7:Exec)
   n2 = Get Skel_Mesh()
   > Set thigh_rot_l(thigh_rot_l=GetAlphaFromReferenceTransformDelta(SkeletalMeshComponent=@n2, BoneName=thigh_l, BaseName=pelvis, InputRange=(X=-20,Y=-100), TwistAxis="1,0,0"))
   > Set thigh_rot_r(thigh_rot_r=GetAlphaFromReferenceTransformDelta(SkeletalMeshComponent=@n2, BoneName=thigh_r, BaseName=pelvis, InputRange=(X=-20,Y=-100), TwistAxis="1,0,0"))
   n3 = Break Fort Anim Input Slope Info!(FortAnimInput_SlopeInfo:"Fort Anim Input Slope Info Structure (by ref)"=Get SlopeInfoInput() -> (SlopeInfoInput:Fort Anim Input Slope Info Structure)) -> (RootSlopeTranslation:Vector=~, LeftFootSlopeTranslation:Vector=~, RightFootSlopeTranslation:Vector=~, SlopeRotation:Rotator=~) {is=title K2Node_BreakStruct}
   Set slope_root_trans(execute=@n1.then_1, slope_root_trans:Vector=@n3) -> (Output_Get:Vector)
   > Set slope_foot_rot(slope_foot_rot:Rotator=@n3.SlopeRotation) -> (Output_Get:Rotator)
   n4 = ComplexSimSpaceFunction DEPRECATED!(
    execute=@n1.then_2
    CharacterPartAnimInstance:Custom Character Part Anim Instance Object Reference=Self-Reference()
    InterpSpeedMax:double=9
    PawnSpeedMax:double=20
    RInterpErrorTolerance:double=1
    DeltaTime:double=0
    DebugDraw:bool=false
    sim_space_rotation_in:Rotator=Get sim_space_rotation() -> (sim_space_rotation:Rotator)
    inComponentSpaceIn:bool=Get InComponentSpace() -> (InComponentSpace:bool)
    inComponentSpaceLastFrameIn:bool=Get InComponentSpaceLastFrame() -> (InComponentSpaceLastFrame:bool)
    lockedOnSimSpaceIn:bool=Get LockedOnSimSpace() -> (LockedOnSimSpace:bool)
    InterpSpeedIn:double=Get InterpSpeed() -> (InterpSpeed:double)
   ) -> (
    then
    sim_space_rotation:Rotator
    inComponentSpaceOut:bool
    inComponentSpaceLastFrameOut:bool
    lockedOnSimSpaceOut:bool
    InterpSpeedOut:double
   )
   > Set sim_space_rotation(sim_space_rotation:Rotator=@n4) -> (Output_Get:Rotator)
   > Set InComponentSpace(InComponentSpace:bool=@n4.inComponentSpaceOut) -> (Output_Get:bool)
   > Set InComponentSpaceLastFrame(InComponentSpaceLastFrame:bool=@n4.inComponentSpaceLastFrameOut) -> (Output_Get:bool)
   > Set LockedOnSimSpace(LockedOnSimSpace:bool=@n4.lockedOnSimSpaceOut) -> (Output_Get:bool)
   > Set InterpSpeed(InterpSpeed:double=@n4.InterpSpeedOut) -> (Output_Get:double)
   n5 = Gravity Override State Function DEPRECATED(
    execute=@n1.then_3
    CharacterPartAnimInstance=Self-Reference()
    GameTimeInSeconds=GetGameTimeInSeconds()
    GravityOverrideParams=Get Gravity Override Params()
   )
   > Set GravityOverride(GravityOverride=@n5)
   n6 = PlayerMovement(execute=@n1.then_4, AnimInstance=Self-Reference(), XDotProductCoefficient=1, YDotProductCoefficient=1, ZDotProductCoefficient=1)
   n7 = float + float(A=float * float(A="Absolute (Float)"(A=@n6), B=0.5), B=MapRangeClamped(Value=@n6.YMovement, InRangeA=0, InRangeB=-1)) {is=title K2Node_CommutativeAssociativeBinaryOperator}
   > Set BwdMovement(BwdMovement:double=MapRangeClamped(Value=@n7, InRangeA=0, InRangeB=1.1, OutRangeA=0.4, OutRangeB=0)) -> (Output_Get:double)
   n8 = MapRangeClamped(Value=@n6, InRangeA=-0.2, InRangeB=1)
   n9 = float * float(A="Absolute (Float)"(A=@n6.YMovement), B=0.2)
   > Set RightMovement(RightMovement:double=MapRangeClamped(Value=float + float(A=@n9, B=@n8), InRangeA=0, InRangeB=1.2, OutRangeB=0.5)) -> (Output_Get:double)
   n10 = MapRangeClamped(Value=@n6, InRangeA=0.2, InRangeB=-1, OutRangeB=0.3)
   n11 = float * float(A="Absolute (Float)"(A=@n6.YMovement), B=0.2)
   > Set LeftMovement(LeftMovement:double=MapRangeClamped(Value=float + float(A=@n11, B=@n10), InRangeA=0, InRangeB=1.2, OutRangeB=0.5)) -> (Output_Get:double)
   > n12 = Set BwdMovement_scarf(BwdMovement_scarf:double=MapRangeClamped(Value=@n7, InRangeA=0, InRangeB=1.1)) -> (Output_Get:double)
   > Set BwdMovement_scarf_0(BwdMovement_scarf_0:bool="float >= float"(A=@n12, B=0.5) {is=title K2Node_CallFunction}) -> (Output_Get:bool)
   n13 = Branch(execute=@n1.then_5, Condition=Get bIsSkyDiving())
   > n14 = Branch(Condition="float > float"(A=Get PawnVelocityZ(), B=50) {is=title K2Node_CallFunction})
   > Set BwdMovement(BwdMovement=float * float(A=Get BwdMovement(), B=0))
   Set BwdMovement(execute=[@n13.else, @n14.else], BwdMovement=float * float(A=Get BwdMovement(), B=1))
   n15 = GetOwnerSkeletalMeshComponent()
   n16 = BreakRotator(InRot=BreakTransform(InTransform=GetDeltaTransformFromRefPose(self=@n15, BoneName=thigh_r, BaseName=pelvis)).Rotation)
   Set thigh_rot_up_r(execute=@n1.then_6, thigh_rot_up_r:double=MapRangeClamped(Value=@n16.Yaw, InRangeA=-15, InRangeB=-100)) -> (Output_Get:double)
   > Set thigh_rot_inr_r(thigh_rot_inr_r:double=MapRangeClamped(Value=@n16.Pitch, InRangeA=0, InRangeB=-10, OutRangeA=1, OutRangeB=0)) -> (Output_Get:double)
   > Set thigh_rot_otr_r(thigh_rot_otr_r:double=MapRangeClamped(Value=@n16.Pitch, InRangeA=20, InRangeB=50)) -> (Output_Get:double)
   n17 = GetOwnerSkeletalMeshComponent()
   n18 = BreakRotator(InRot=BreakTransform(InTransform=GetDeltaTransformFromRefPose(self=@n17, BoneName=thigh_l, BaseName=pelvis)).Rotation)
   > Set thigh_rot_up_l(thigh_rot_up_l:double=MapRangeClamped(Value=@n18.Yaw, InRangeA=-10, InRangeB=-100)) -> (Output_Get:double)
   > Set thigh_rot_inr_l(thigh_rot_inr_l:double=MapRangeClamped(Value=@n18.Pitch, InRangeA=0, InRangeB=-10, OutRangeA=1, OutRangeB=0)) -> (Output_Get:double)
   > Set thigh_rot_otr_l(thigh_rot_otr_l:double=MapRangeClamped(Value=@n18.Pitch, InRangeA=20, InRangeB=50)) -> (Output_Get:double)
   > Set jumpingAlpha(jumpingAlpha:double="To Float (Boolean)"(InBool=Get bIsJumping())) -> (Output_Get:double)
   n19 = Get bIsMoving2D()
   > Set isNotCrouchRunning(isNotCrouchRunning:double="To Float (Boolean)"(InBool=NOT Boolean(A=@n19))) -> (Output_Get:double)
   > Set ParachuteAlpha(ParachuteAlpha:double="To Float (Boolean)"(InBool=Get bIsParachuteOpened())) -> (Output_Get:double)
   > Set skydiveAlpha(skydiveAlpha:double="To Float (Boolean)"(InBool=Get bIsSkyDiving())) -> (Output_Get:double)
   n20 = Sequence!(execute=@n1.then_7) -> (then_0, then_1, then_2, then_3)
   > n21 = Branch(Condition=Get bIsFerretVehicleDriver() -> (bIsFerretVehicleDriver:bool))
   Set in vehicle rigid alpha(execute=[@n21.then, @n22.then], in vehicle rigid alpha:double=0) -> (Output_Get:double)
   > Set in vehicle scale alpha(in vehicle scale alpha:double=1) -> (Output_Get:double)
   n22 = Branch(execute=@n21.else, Condition=Get bIsOctopusVehicleDriver() -> (bIsOctopusVehicleDriver:bool))
   Set in vehicle rigid alpha(execute=@n22.else, in vehicle rigid alpha=1)
   > Set in vehicle scale alpha(in vehicle scale alpha=0)
   n23 = Branch(execute=@n20.then_1, Condition=CharacterItemDefinitionHasAnyMetaTag(Tags=Make Array("[0]"=(TagName="Cosmetics.Variant.Property.Stage0"))))
   > Set ScarfRbanAlpha(ScarfRbanAlpha:double=0) -> (Output_Get:double)
   > Set CoatRbanAlpha(CoatRbanAlpha:double=0) -> (Output_Get:double)
   > Set LegTasselsAnimDynAlpha(LegTasselsAnimDynAlpha:double=0) -> (Output_Get:double)
   > Set LegStrapsAnimDynAlpha(LegStrapsAnimDynAlpha:double=1) -> (Output_Get:double)
   > Set ArmTasselsAnimDynAlpha(ArmTasselsAnimDynAlpha:double=0) -> (Output_Get:double)
   > Set SpursRbanAlpha(SpursRbanAlpha:double=0) -> (Output_Get:double)
   n24 = Branch(execute=@n23.else, Condition=CharacterItemDefinitionHasAnyMetaTag(Tags=Make Array("[0]"=(TagName="Cosmetics.Variant.Property.Stage1"))))
   > Set ScarfRbanAlpha(ScarfRbanAlpha=0)
   > Set CoatRbanAlpha(CoatRbanAlpha=0)
   > Set LegTasselsAnimDynAlpha(LegTasselsAnimDynAlpha=0)
   > Set LegStrapsAnimDynAlpha(LegStrapsAnimDynAlpha=0)
   > Set ArmTasselsAnimDynAlpha(ArmTasselsAnimDynAlpha=0)
   > Set SpursRbanAlpha(SpursRbanAlpha=0)
   n25 = Branch(execute=@n24.else, Condition=CharacterItemDefinitionHasAnyMetaTag(Tags=Make Array("[0]"=(TagName="Cosmetics.Variant.Property.Stage2"))))
   > Set ScarfRbanAlpha(ScarfRbanAlpha=0)
   > Set CoatRbanAlpha(CoatRbanAlpha=0)
   > Set LegTasselsAnimDynAlpha(LegTasselsAnimDynAlpha=1)
   > Set LegStrapsAnimDynAlpha(LegStrapsAnimDynAlpha=0)
   > Set ArmTasselsAnimDynAlpha(ArmTasselsAnimDynAlpha=1)
   > Set SpursRbanAlpha(SpursRbanAlpha=0)
   n26 = Branch(execute=@n25.else, Condition=CharacterItemDefinitionHasAnyMetaTag(Tags=Make Array("[0]"=(TagName="Cosmetics.Variant.Property.Stage3"))))
   > Set ScarfRbanAlpha(ScarfRbanAlpha=1)
   > Set CoatRbanAlpha(CoatRbanAlpha=0)
   > Set LegTasselsAnimDynAlpha(LegTasselsAnimDynAlpha=1)
   > Set LegStrapsAnimDynAlpha(LegStrapsAnimDynAlpha=0)
   > Set ArmTasselsAnimDynAlpha(ArmTasselsAnimDynAlpha=1)
   > Set SpursRbanAlpha(SpursRbanAlpha=1)
   n27 = Branch(execute=@n26.else, Condition=CharacterItemDefinitionHasAnyMetaTag(Tags=Make Array("[0]"=(TagName="Cosmetics.Variant.Property.Stage4"))))
   > Set ScarfRbanAlpha(ScarfRbanAlpha=1)
   > Set CoatRbanAlpha(CoatRbanAlpha=1)
   > Set LegTasselsAnimDynAlpha(LegTasselsAnimDynAlpha=0)
   > Set LegStrapsAnimDynAlpha(LegStrapsAnimDynAlpha=0)
   > Set ArmTasselsAnimDynAlpha(ArmTasselsAnimDynAlpha=1)
   > Set SpursRbanAlpha(SpursRbanAlpha=1)
   Branch(execute=@n27.else, Condition=CharacterItemDefinitionHasAnyMetaTag(Tags=Make Array("[0]"=(TagName="Cosmetics.Variant.Property.Stage5"))))
   > Set ScarfRbanAlpha(ScarfRbanAlpha=1)
   > Set CoatRbanAlpha(CoatRbanAlpha=1)
   > Set LegTasselsAnimDynAlpha(LegTasselsAnimDynAlpha=0)
   > Set LegStrapsAnimDynAlpha(LegStrapsAnimDynAlpha=0)
   > Set ArmTasselsAnimDynAlpha(ArmTasselsAnimDynAlpha=0)
   > Set SpursRbanAlpha(SpursRbanAlpha=1)
   Set ScarfRbanAlphaFINAL(
    execute=@n20.then_2
    ScarfRbanAlphaFINAL:double=float * float(A=Get ScarfRbanAlpha() -> (ScarfRbanAlpha:double), B=Get dj_emote_alpha() -> (dj_emote_alpha:double))
   ) -> (Output_Get:double)
   > Set CoatRbanAlphaFINAL(CoatRbanAlphaFINAL:double=float * float(A=Get CoatRbanAlpha() -> (CoatRbanAlpha:double), B=Get in vehicle rigid alpha() -> (in vehicle rigid alpha:double))) -> (Output_Get:double)
   n28 = Event OnNotifyPartsStarted!() -> (OutputDelegate, then, Context:Name=None) {is=title K2Node_Event}
   > Branch(Condition="Equal (Name)"!(A:Name=@n28.Context, B:Name=Emote_DJ_Drop) -> (ReturnValue:bool) {is=title K2Node_CallFunction})
   > Set dj_emote_alpha(dj_emote_alpha:double=0) -> (Output_Get:double)
   n29 = Event OnNotifyPartsStopped!() -> (OutputDelegate, then, Context:Name=None) {is=title K2Node_Event}
   > Branch(Condition="Equal (Name)"(A=@n29.Context, B=Emote_DJ_Drop))
   > Set dj_emote_alpha(dj_emote_alpha=1)
   Event BlueprintInitializeAnimation()
   > Set Skel_Mesh(Skel_Mesh=GetOwnerSkeletalMeshComponent())
   > Set InterpSpeed(InterpSpeed=Get InterpSpeedMax() -> (InterpSpeedMax:double))
   n30 = Event OnBeginTargeting() {is=title K2Node_Event}
   n31 = Get bIsCrouching()
   n32 = Branch(execute=[@n30.then, @n35.then], Condition=AND Boolean(A=Get bIsTargeting(), B=NOT Boolean(A=@n31)))
   ResetDynamics(execute=[@n33.then, @n34.then, @n36.then, @n32.then], InTeleportType=TeleportPhysics)
   > ResetDynamics(InTeleportType=ResetPhysics)
   Event OnParachuteOpened()
   > n33 = Delay(Duration=1.5)
   n34 = Event OnEndSkydiving()
   n35 = Event OnEndCrouching()
   n36 = Event OnBeginSkydiving()
   Event OnCharacterCustomizationLoaded()
   Get bIsMoving2D()
   Get bIsCrouchSprinting()
  ]}
]
variables [
 ArmTasselsAnimDynAlpha {type=number}
 BwdMovement {type=number}
 BwdMovement_scarf {type=number}
 BwdMovement_scarf_0 {type=boolean}
 CoatRbanAlpha {type=number}
 CoatRbanAlphaFINAL {type=number}
 CurrentStage {type=integer}
 dj_emote_alpha {type=number, default=1}
 Gravity Override Params {type=GravityOverrideParamsStruct}
 GravityOverride {type=Vector}
 in vehicle rigid alpha {type=number, default=1}
 in vehicle scale alpha {type=number}
 InComponentSpace {type=boolean}
 InComponentSpaceLastFrame {type=boolean}
 InterpSpeed {type=number}
 InterpSpeedMax {type=number}
 isNotCrouchRunning {type=number}
 jumpingAlpha {type=number}
 LeftMovement {type=number}
 LegStrapsAnimDynAlpha {type=number}
 LegTasselsAnimDynAlpha {type=number}
 LockedOnSimSpace {type=boolean}
 ParachuteAlpha {type=number}
 RightMovement {type=number}
 ScarfRbanAlpha {type=number}
 ScarfRbanAlphaFINAL {type=number}
 sim_space_rotation {type=Rotator}
 Skel_Mesh {type=SkeletalMeshComponent}
 skydiveAlpha {type=number}
 slope_foot_rot {type=Rotator}
 slope_root_trans {type=Vector}
 SpursRbanAlpha {type=number}
 thigh_rot_inr_l {type=number}
 thigh_rot_inr_r {type=number}
 thigh_rot_l {type=number}
 thigh_rot_otr_l {type=number}
 thigh_rot_otr_r {type=number}
 thigh_rot_r {type=number}
 thigh_rot_up_l {type=number}
 thigh_rot_up_r {type=number}
]
